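# Physics asset for the Teriyaki Fish Assassin back bling, giving its dangling hair, arm and leg bones ragdoll-style secondary motion.
# Capsule bodies hang off a single large pack capsule, with each arm built from two capsules.
# Swing and twist limited constraints tie the arms, a branching hair cluster and two three-segment leg chains back to the pack.
# The hair root and arms have tighter limits than the hair tips and legs.
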PhysicsAsset /BRCosmetics/Accessories/FORT_Backpacks/Backpack_M_MED_Teriyaki_Fish_Assassin/Meshes/M_MED_Teriyaki_Fish_Assassin_Physics
bodies {
 dyn_arm_01_l [
  {shapeType=Capsule, length=2.188, radius=3.7233, center=<2.7229 0 -0.531>, rotation=<rot 61.3158 180 179.9999>}
  {
   shapeType=Capsule
   length=10.875
   radius=1.625
   shapeName=dyn_arm_01_l_capsule
   center=<11.7411 -0.4564 6.9637>
   rotation=<rot -63.3047 -2.4127 17.2991>
  }
 ]
 dyn_arm_01_r [
  {shapeType=Capsule, length=2.563, radius=3.5983, center=<2.165 0 0.4505>, rotation=<rot 64.6883 -0.0001 -0.0001>}
  {
   shapeType=Capsule
   length=11.125
   radius=1.75
   shapeName=dyn_arm_01_r_capsule
   center=<10.7522 -1.5746 -6.6176>
   rotation=<rot 62.8407 -3.1803 -18.3603>
  }
 ]
 dyn_hair_01 [{shapeType=Capsule, length=0.9553, radius=2.3596, center=<0.1 0 0>}]
 dyn_hair_02 [{shapeType=Capsule, length=0.312, radius=0.693, center=<0.1 0 0>}]
 dyn_hair_03 [{shapeType=Capsule, length=0.505, radius=0.5, center=<0.1 0 0>}]
 dyn_hair_04 [{shapeType=Capsule, length=0.4155, radius=0.5, center=<0.1 0 0.0712>}]
 dyn_leg_01_l [{shapeType=Capsule, length=2, radius=1, center=<1 0 0>, rotation=<rot 90 0 0>}]
 dyn_leg_01_r [{shapeType=Capsule, length=2, radius=1, center=<-1 0 0>, rotation=<rot 90 0 0>}]
 dyn_leg_02_l [{shapeType=Capsule, length=2, radius=1, center=<1 0 0>, rotation=<rot 90 0 0>}]
 dyn_leg_02_r [{shapeType=Capsule, length=2, radius=1, center=<-1 0 0>, rotation=<rot 90 0 0>}]
 dyn_leg_03_l [{shapeType=Capsule, length=2, radius=1, center=<1 0 0>, rotation=<rot 90 0 0>}]
 dyn_leg_03_r [{shapeType=Capsule, length=2, radius=1, center=<-1 0 0>, rotation=<rot 90 0 0>}]
 pack [{shapeType=Capsule, length=0.2099, radius=10.5, center=<0.0409 -8.9298 -0.259>, rotation=<rot 1.6673 89.9668 -91.1403>}]
}
constraints [
 {bone1Name=dyn_hair_01, bone2Name=pack, swing1LimitDegrees=30, swing2LimitDegrees=30, twistLimitDegrees=5}
 {bone1Name=dyn_hair_02, bone2Name=dyn_hair_01, swing1LimitDegrees=45, swing2LimitDegrees=45, twistLimitDegrees=15}
 {bone1Name=dyn_hair_03, bone2Name=dyn_hair_01, swing1LimitDegrees=45, swing2LimitDegrees=45, twistLimitDegrees=15}
 {bone1Name=dyn_hair_04, bone2Name=dyn_hair_01, swing1LimitDegrees=45, swing2LimitDegrees=45, twistLimitDegrees=15}
 {bone1Name=dyn_leg_01_l, bone2Name=pack, swing1LimitDegrees=45, swing2LimitDegrees=45, twistLimitDegrees=15}
 {bone1Name=dyn_leg_02_l, bone2Name=dyn_leg_01_l, swing1LimitDegrees=45, swing2LimitDegrees=45, twistLimitDegrees=15}
 {bone1Name=dyn_leg_03_l, bone2Name=dyn_leg_02_l, swing1LimitDegrees=45, swing2LimitDegrees=45, twistLimitDegrees=15}
 {bone1Name=dyn_leg_01_r, bone2Name=pack, swing1LimitDegrees=45, swing2LimitDegrees=45, twistLimitDegrees=15}
 {bone1Name=dyn_leg_02_r, bone2Name=dyn_leg_01_r, swing1LimitDegrees=45, swing2LimitDegrees=45, twistLimitDegrees=15}
 {bone1Name=dyn_leg_03_r, bone2Name=dyn_leg_02_r, swing1LimitDegrees=45, swing2LimitDegrees=45, twistLimitDegrees=15}
 {bone1Name=dyn_arm_01_l, bone2Name=pack, swing1LimitDegrees=35, swing2LimitDegrees=35, twistLimitDegrees=10}
 {bone1Name=dyn_arm_01_r, bone2Name=pack, swing1LimitDegrees=35, swing2LimitDegrees=35, twistLimitDegrees=10}
]
