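# Animation Blueprint for the Teriyaki Fish Assassin backpack cosmetic that adds secondary motion and idle life to its eyes and binky.
# A RigidBody node simulates physics over the reference pose, then additive bone edits offset the binky and rotate both eyes together.
# Each update bobs the binky on a time-driven sine wave with interpolated amplitude.
# Chained Delays in the update event periodically randomize eye rotation, then the binky's target range.
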AnimBlueprint /BRCosmetics/Accessories/FORT_Backpacks/Backpack_M_MED_Teriyaki_Fish_Assassin/Meshes/M_MED_Teriyaki_Fish_Assassin_AnimBP
target_skeleton=&/M_MED_Teriyaki_Fish_Assassin_Skeleton
graphs [
 AnimGraph {nodes [
   n1 = RigidBody(ComponentPose=Local To Component(LocalPose=Local Space Ref Pose())) {
    bOverrideWorldGravity=true
    overridePhysicsAsset=&/M_MED_Teriyaki_Fish_Assassin_Physics
    simulationSpace=BaseBoneSpace
    alphaBoolBlend {blendOption=Linear}
    baseBoneRef {boneName=root}
    componentLinearAccScale=<0.4 0.4 0.4>
    componentLinearVelScale=<0.2 0.2 0.2>
    overrideWorldGravity {z=-980}
   }
   n2 = Get eyeRotation() -> (eyeRotation:Rotator)
   n3 = "Transform (Modify) Bone - Bone: eye_l"(
    ComponentPose="Transform (Modify) Bone - Bone: binky"(ComponentPose=@n1, Alpha=1, Translation=Get binkyVect() -> (binkyVect:Vector), Rotation="0,0,0", Scale="1,1,1") {translationMode=BMM_Additive, alphaBoolBlend {blendOption=Linear}, boneToModify {boneName=binky}}
    Alpha=1
    Translation="0,0,0"
    Rotation=@n2
    Scale="1,1,1"
   ) {rotationMode=BMM_Additive, alphaBoolBlend {blendOption=Linear}, boneToModify {boneName=eye_l}}
   Output Pose(Result=Component To Local(ComponentPose="Transform (Modify) Bone - Bone: eye_r"(ComponentPose=@n3, Alpha=1, Translation="0,0,0", Rotation=@n2, Scale="1,1,1") {rotationMode=BMM_Additive, alphaBoolBlend {blendOption=Linear}, boneToModify {boneName=eye_r}})) {layerGroup=None}
  ]}
 EventGraph {nodes [
   n1 = Event BlueprintUpdateAnimation()
   > n2 = Set binkyRange(binkyRange:double=FInterpTo(
     Current=Get binkyRange() -> (binkyRange:double)
     Target=Get targetBinkyRange() -> (targetBinkyRange:double)
     DeltaTime=@n1.DeltaTimeX
     InterpSpeed=7
    )) -> (Output_Get:double)
   n3 = GetTimeSeconds!() -> (ReturnValue:double)
   n4 = MapRangeClamped(Value="Sin (Radians)"(A=float * float(A=@n3, B=7)), InRangeA=-1, InRangeB=1)
   > Set binkyVect(binkyVect:Vector=MakeVector(X=Lerp(A=@n2, B=0, Alpha=@n4), Y=Lerp(A=-0.07, B=0, Alpha=@n4))) -> (Output_Get:Vector)
   > Delay(Duration=RandomFloatInRange(Min=1, Max=5))
   > Set eyeRotation(eyeRotation:Rotator=MakeRotator(Roll=RandomFloatInRange(Min=-25, Max=25), Yaw=RandomFloatInRange(Min=-30, Max=30))) -> (Output_Get:Rotator)
   > Delay(Duration=1.5)
   > Set targetBinkyRange(targetBinkyRange:double=RandomFloatInRange(Min=-1, Max=0)) -> (Output_Get:double)
   TryGetPawnOwner()
  ]}
]
variables [
 binkyRange {type=number, default=-1}
 binkyVect {type=Vector, default=<0 0 1>}
 eyeRotation {type=Rotator}
 targetBinkyRange {type=number}
]
